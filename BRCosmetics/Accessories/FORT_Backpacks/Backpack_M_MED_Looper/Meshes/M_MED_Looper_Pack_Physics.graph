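# Physics asset for the M_MED_Looper backpack cosmetic, providing bodies and joints for the bag and its five dyn_wire bones.
# Each bone gets one small capsule, with the wire capsules offset and rotated from their bone origins.
# Every wire is constrained directly to the bag rather than chained to another wire.
# Wires 01–03 have wider swing and twist limits, while 04–05 are held tighter.
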PhysicsAsset /BRCosmetics/Accessories/FORT_Backpacks/Backpack_M_MED_Looper/Meshes/M_MED_Looper_Pack_Physics
bodies {
 bag [{shapeType=Capsule, length=1, radius=1, shapeName=bag_capsule}]
 dyn_wire01 [{shapeType=Capsule, length=1, radius=1, shapeName=dyn_wire01_capsule, center=<1 0 0>, rotation=<rot 90 0 0>}]
 dyn_wire02 [{shapeType=Capsule, length=1, radius=1, shapeName=dyn_wire02_capsule, center=<1 0 0>, rotation=<rot 90 0 0>}]
 dyn_wire03 [{shapeType=Capsule, length=1, radius=1, shapeName=dyn_wire03_capsule, center=<1 0 0>, rotation=<rot 90 0 0>}]
 dyn_wire04 [{shapeType=Capsule, length=1, radius=1, shapeName=dyn_wire04_capsule, center=<1 0 0>, rotation=<rot 90 0 0>}]
 dyn_wire05 [{shapeType=Capsule, length=1, radius=1, shapeName=dyn_wire05_capsule, center=<1 0 0>, rotation=<rot 90 0 0>}]
}
constraints [
 {bone1Name=dyn_wire01, bone2Name=bag, swing1LimitDegrees=45, swing2LimitDegrees=45, twistLimitDegrees=45}
 {bone1Name=dyn_wire02, bone2Name=bag, swing1LimitDegrees=45, swing2LimitDegrees=45, twistLimitDegrees=45}
 {bone1Name=dyn_wire03, bone2Name=bag, swing1LimitDegrees=45, swing2LimitDegrees=45, twistLimitDegrees=45}
 {bone1Name=dyn_wire04, bone2Name=bag, swing1LimitDegrees=20, swing2LimitDegrees=20, twistLimitDegrees=10}
 {bone1Name=dyn_wire05, bone2Name=bag, swing1LimitDegrees=20, swing2LimitDegrees=20, twistLimitDegrees=10}
]
