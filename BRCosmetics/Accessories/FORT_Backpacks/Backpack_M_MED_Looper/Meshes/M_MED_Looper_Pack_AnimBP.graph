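# Animation Blueprint for the Looper medkit backpack cosmetic, animating the pack with idle and box-trick motions.
# The output randomly plays idle or tricks, adds jitter, nudges the pack bone, then runs RigidBody physics with gravity updated each frame.
# The EventGraph pulses ReactiveAnimation on chest opens, eliminations or SlurpJuice shield gains.
# Its Idle and Reactive state machine and two testing machines are not wired to the output.
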AnimBlueprint /BRCosmetics/Accessories/FORT_Backpacks/Backpack_M_MED_Looper/Meshes/M_MED_Looper_Pack_AnimBP
target_skeleton=/Game/Accessories/FORT_Backpacks/Backpack_M_MED_Looper/Meshes/M_MED_Looper_Pack_Skeleton
graphs [
 AnimGraph {nodes [
   Input Pose()
   n1 = Random Sequence Player!() -> (Pose:Pose Link Structure) {
    is=title AnimGraphNode_RandomPlayer
    entries [
     {
      chanceToPlay=0.7
      maxLoopCount=0
      maxPlayRate=1
      minLoopCount=0
      minPlayRate=1
      sequence=/BRCosmetics/Accessories/FORT_Backpacks/Backpack_M_MED_Looper/Animations/Looper_Idle_Medkit_PACK
      blendIn {blendOption=Linear, blendTime=0.2, customCurve=null}
     }
     {
      chanceToPlay=0.3
      maxLoopCount=0
      maxPlayRate=1
      minLoopCount=0
      minPlayRate=1
      sequence=/BRCosmetics/Accessories/FORT_Backpacks/Backpack_M_MED_Looper/Animations/Looper_BoxTrick_01_PACK
      blendIn {blendOption=Linear, blendTime=0.2, customCurve=null}
     }
     {
      chanceToPlay=0.3
      maxLoopCount=0
      maxPlayRate=1
      minLoopCount=0
      minPlayRate=1
      sequence=/BRCosmetics/Accessories/FORT_Backpacks/Backpack_M_MED_Looper/Animations/Looper_BoxTrick_02_PACK
      blendIn {blendOption=Linear, blendTime=0.2, customCurve=null}
     }
    ]
   }
   n2 = "Sequence Player (additive) 'Looper_Jitter_PACK'"() {sequence=/BRCosmetics/Accessories/FORT_Backpacks/Backpack_M_MED_Looper/Animations/Looper_Jitter_PACK}
   n3 = "Transform (Modify) Bone - Bone: pack"(
    ComponentPose=Local To Component(LocalPose=Apply Additive(Base=@n1, Additive=@n2) {alphaBoolBlend {blendOption=Linear}})
    Translation="0,-0.5,0"
    Rotation="0,0,0"
   ) {translationMode=BMM_Additive, alphaBoolBlend {blendOption=Linear}, boneToModify {boneName=pack}, translation {y=-0.5}}
   n4 = RigidBody!(ComponentPose=@n3, Alpha, OverrideWorldGravity:Vector=Get GravityOverride(), ExternalForce) -> (Pose) {
    bClampLinearTranslationLimitToRefPose=true
    bOverrideWorldGravity=true
    overridePhysicsAsset=&/M_MED_Looper_Pack_Physics
    simulationSpace=BaseBoneSpace
    alphaBoolBlend {blendOption=Linear}
    baseBoneRef {boneName=root}
    componentAppliedLinearAccClamp=<3000 3000 3000>
    componentLinearAccScale=<0.4 0.3 0.3>
    componentLinearVelScale=<0.4 0.3 0.3>
    overrideWorldGravity {z=-980}
    simSpaceSettings {
     maxAngularAcceleration=1000
     maxAngularVelocity=3
     maxLinearAcceleration=700
     maxLinearVelocity=500
     velocityScaleZ=0
     worldAlpha=1
     externalLinearDragV=<0.2 0.2>
    }
   }
   Output Pose(Result=Component To Local(ComponentPose=@n4)) {layerGroup=None}
   AnimGraphNode_StateMachine_0 = "Animation Testing (BoxTricks)"() {is=title AnimGraphNode_StateMachine}
   AnimGraphNode_StateMachine_1 = Idle and Reactive() {is=title AnimGraphNode_StateMachine}
   AnimGraphNode_StateMachine_2 = "Animation Testing (GetGive Pack)"() {is=title AnimGraphNode_StateMachine}
  ]}
 "Animation Testing (BoxTricks)" {
  entry=State
  path="AnimGraph.AnimGraphNode_StateMachine_0.Animation Testing (BoxTricks)"
  ids=id
  nodes [
   AnimStateNode_0 = State_2() {is=title AnimStateNode}
   AnimStateNode_2 = State() {is=title AnimStateNode}
   AnimStateNode_3 = State_1() {is=title AnimStateNode}
   AnimStateTransitionNode_2 = AnimStateTransitionNode() {bAutomaticRuleBasedOnSequencePlayerInState=true}
   AnimStateTransitionNode_3 = AnimStateTransitionNode() {bAutomaticRuleBasedOnSequencePlayerInState=true}
   AnimStateTransitionNode_4 = AnimStateTransitionNode() {bAutomaticRuleBasedOnSequencePlayerInState=true}
  ]
  links [
   {
    dst=State_2
    rule="AnimGraph.AnimGraphNode_StateMachine_0.Animation Testing (BoxTricks).AnimStateTransitionNode_2.Transition"
    src=State
    via=AnimStateTransitionNode_2
   }
   {
    dst=State
    rule="AnimGraph.AnimGraphNode_StateMachine_0.Animation Testing (BoxTricks).AnimStateTransitionNode_4.Transition"
    src=State_1
    via=AnimStateTransitionNode_4
   }
   {
    dst=State_1
    rule="AnimGraph.AnimGraphNode_StateMachine_0.Animation Testing (BoxTricks).AnimStateTransitionNode_3.Transition"
    src=State_2
    via=AnimStateTransitionNode_3
   }
  ]
 }
 State_2 {
  path="AnimGraph.AnimGraphNode_StateMachine_0.Animation Testing (BoxTricks).AnimStateNode_0.State_2"
  nodes [Output Animation Pose(Result=Sequence Player 'Looper_BoxTrick_02_PACK'() {bLoopAnimation=false, sequence=&Looper_BoxTrick_02_PACK})]
 }
 State {
  path="AnimGraph.AnimGraphNode_StateMachine_0.Animation Testing (BoxTricks).AnimStateNode_2.State"
  nodes [Output Animation Pose(Result=Sequence Player 'Looper_Idle_Medkit_PACK'() {bLoopAnimation=false, sequence=&Looper_Idle_Medkit_PACK})]
 }
 State_1 {
  path="AnimGraph.AnimGraphNode_StateMachine_0.Animation Testing (BoxTricks).AnimStateNode_3.State_1"
  nodes [Output Animation Pose(Result=Sequence Player 'Looper_BoxTrick_01_PACK'() {bLoopAnimation=false, sequence=&Looper_BoxTrick_01_PACK})]
 }
 Transition {
  path="AnimGraph.AnimGraphNode_StateMachine_0.Animation Testing (BoxTricks).AnimStateTransitionNode_2.Transition"
  nodes [Result(bCanEnterTransition=true)]
 }
 Transition {
  path="AnimGraph.AnimGraphNode_StateMachine_0.Animation Testing (BoxTricks).AnimStateTransitionNode_3.Transition"
  nodes [Result(bCanEnterTransition=true)]
 }
 Transition {path="AnimGraph.AnimGraphNode_StateMachine_0.Animation Testing (BoxTricks).AnimStateTransitionNode_4.Transition", nodes [Result()]}
 Idle and Reactive {
  entry=Idle
  path=AnimGraph.AnimGraphNode_StateMachine_1.Idle and Reactive
  ids=id
  nodes [
   AnimStateNode_2 = Idle() {
    stateEntered {guid=#1}
    stateFullyBlended {guid=#2}
    stateLeft {guid=#3}
   }
   AnimStateNode_3 = Reactive_Anim() {
    is=title AnimStateNode
    stateEntered {guid=#4}
    stateFullyBlended {guid=#5}
    stateLeft {guid=#6}
   }
   AnimStateTransitionNode_2 = AnimStateTransitionNode() {
    crossfadeDuration=0.3
    transitionEnd {guid=#7}
    transitionInterrupt {guid=#8}
    transitionStart {guid=#9}
   }
   AnimStateTransitionNode_3 = AnimStateTransitionNode() {
    bAutomaticRuleBasedOnSequencePlayerInState=true
    crossfadeDuration=0.1
    transitionEnd {guid=#10}
    transitionInterrupt {guid=#11}
    transitionStart {guid=#12}
   }
  ]
  links [
   {
    dst=Reactive_Anim
    rule=AnimGraph.AnimGraphNode_StateMachine_1.Idle and Reactive.AnimStateTransitionNode_2.Transition
    src=Idle
    via=AnimStateTransitionNode_2
   }
   {
    dst=Idle
    rule=AnimGraph.AnimGraphNode_StateMachine_1.Idle and Reactive.AnimStateTransitionNode_3.Transition
    src=Reactive_Anim
    via=AnimStateTransitionNode_3
   }
  ]
 }
 Idle {
  path=AnimGraph.AnimGraphNode_StateMachine_1.Idle and Reactive.AnimStateNode_2.Idle
  nodes [Output Animation Pose(Result=Sequence Player 'Looper_Idle_Medkit_PACK'() {sequence=&Looper_Idle_Medkit_PACK})]
 }
 Reactive_Anim {
  path=AnimGraph.AnimGraphNode_StateMachine_1.Idle and Reactive.AnimStateNode_3.Reactive_Anim
  nodes [
   n1 = Random Sequence Player() {entries [
     {
      chanceToPlay=0.5
      maxLoopCount=0
      maxPlayRate=0.8
      minLoopCount=0
      minPlayRate=1
      sequence=&Looper_BoxTrick_01_PACK
      blendIn {blendOption=Linear, blendTime=0.2, customCurve=null}
     }
     {
      chanceToPlay=0.5
      maxLoopCount=0
      maxPlayRate=1
      minLoopCount=0
      minPlayRate=1
      sequence=&Looper_BoxTrick_02_PACK
      blendIn {blendOption=Linear, blendTime=0.2, customCurve=null}
     }
    ]}
   Output Animation Pose(Result=@n1)
  ]
 }
 Transition {
  path=AnimGraph.AnimGraphNode_StateMachine_1.Idle and Reactive.AnimStateTransitionNode_2.Transition
  nodes [Result(bCanEnterTransition=Get ReactiveAnimation() -> (ReactiveAnimation:bool))]
 }
 Transition {path=AnimGraph.AnimGraphNode_StateMachine_1.Idle and Reactive.AnimStateTransitionNode_3.Transition, nodes [Result()]}
 "Animation Testing (GetGive Pack)" {
  entry=State
  path="AnimGraph.AnimGraphNode_StateMachine_2.Animation Testing (GetGive Pack)"
  ids=id
  nodes [
   AnimStateNode_0 = State_2()
   AnimStateNode_1 = State_1_1() {is=title AnimStateNode}
   AnimStateNode_2 = State()
   AnimStateNode_3 = State_1()
   AnimStateTransitionNode_0 = AnimStateTransitionNode() {bAutomaticRuleBasedOnSequencePlayerInState=true}
   AnimStateTransitionNode_1 = AnimStateTransitionNode() {bAutomaticRuleBasedOnSequencePlayerInState=true}
   AnimStateTransitionNode_2 = AnimStateTransitionNode() {bAutomaticRuleBasedOnSequencePlayerInState=true}
   AnimStateTransitionNode_3 = AnimStateTransitionNode() {bAutomaticRuleBasedOnSequencePlayerInState=true}
  ]
  links [
   {
    dst=State_2
    rule="AnimGraph.AnimGraphNode_StateMachine_2.Animation Testing (GetGive Pack).AnimStateTransitionNode_2.Transition"
    src=State
    via=AnimStateTransitionNode_2
   }
   {
    dst=State_1_1
    rule="AnimGraph.AnimGraphNode_StateMachine_2.Animation Testing (GetGive Pack).AnimStateTransitionNode_0.Transition"
    src=State_1
    via=AnimStateTransitionNode_0
   }
   {
    dst=State
    rule="AnimGraph.AnimGraphNode_StateMachine_2.Animation Testing (GetGive Pack).AnimStateTransitionNode_1.Transition"
    src=State_1_1
    via=AnimStateTransitionNode_1
   }
   {
    dst=State_1
    rule="AnimGraph.AnimGraphNode_StateMachine_2.Animation Testing (GetGive Pack).AnimStateTransitionNode_3.Transition"
    src=State_2
    via=AnimStateTransitionNode_3
   }
  ]
 }
 State_2 {
  path="AnimGraph.AnimGraphNode_StateMachine_2.Animation Testing (GetGive Pack).AnimStateNode_0.State_2"
  nodes [Output Animation Pose(Result=Sequence Player 'Looper_GetMedkit_PACK'() {bLoopAnimation=false, sequence=/Game/Accessories/FORT_Backpacks/Backpack_M_MED_Looper/Animations/Looper_GetMedkit_PACK})]
 }
 State_1_1 {
  path="AnimGraph.AnimGraphNode_StateMachine_2.Animation Testing (GetGive Pack).AnimStateNode_1.State_1_1"
  nodes [Output Animation Pose(Result=Sequence Player 'Looper_GiveMedkit_PACK'() {bLoopAnimation=false, sequence=/Game/Accessories/FORT_Backpacks/Backpack_M_MED_Looper/Animations/Looper_GiveMedkit_PACK})]
 }
 State {
  path="AnimGraph.AnimGraphNode_StateMachine_2.Animation Testing (GetGive Pack).AnimStateNode_2.State"
  nodes [Output Animation Pose(Result=Sequence Player 'Looper_Idle_PACK'() {bLoopAnimation=false, sequence=/Game/Accessories/FORT_Backpacks/Backpack_M_MED_Looper/Animations/Looper_Idle_PACK})]
 }
 State_1 {
  path="AnimGraph.AnimGraphNode_StateMachine_2.Animation Testing (GetGive Pack).AnimStateNode_3.State_1"
  nodes [Output Animation Pose(Result=Sequence Player 'Looper_Idle_Medkit_PACK'() {bLoopAnimation=false, sequence=&Looper_Idle_Medkit_PACK})]
 }
 Transition {path="AnimGraph.AnimGraphNode_StateMachine_2.Animation Testing (GetGive Pack).AnimStateTransitionNode_0.Transition", nodes [Result()]}
 Transition {path="AnimGraph.AnimGraphNode_StateMachine_2.Animation Testing (GetGive Pack).AnimStateTransitionNode_1.Transition", nodes [Result()]}
 Transition {
  path="AnimGraph.AnimGraphNode_StateMachine_2.Animation Testing (GetGive Pack).AnimStateTransitionNode_2.Transition"
  nodes [Result(bCanEnterTransition=true)]
 }
 Transition {
  path="AnimGraph.AnimGraphNode_StateMachine_2.Animation Testing (GetGive Pack).AnimStateTransitionNode_3.Transition"
  nodes [Result(bCanEnterTransition=true)]
 }
 EventGraph {nodes [
   Event BlueprintUpdateAnimation()
   > n1 = Gravity Override State Function DEPRECATED(CharacterPartAnimInstance=Self-Reference(), GameTimeInSeconds=GetGameTimeInSeconds(), GravityOverrideParams=Get Gravity Override Params())
   > Set GravityOverride(GravityOverride=@n1)
   n2 = Event BlueprintProcessAnimInputEvent()
   n3 = IsNamed(self=@n2.AnimInputEvent, InEventName=aie.PlayerVitality.ShieldsIncreased)
   n4 = Get As Player Pawn Athena() -> (As Player Pawn Athena:Player Pawn Athena Object Reference)
   n5 = Contains(
    SearchIn=GetDisplayName(Object=Get CurrentWeapon!(self:Fort Pawn Object Reference=@n4) -> (CurrentWeapon:Fort Weapon Object Reference))
    Substring=SlurpJuice
   )
   > Branch(Condition=OR Boolean!(
     A=IsNamed(self=@n2.AnimInputEvent, InEventName=aie.ChestOpened)
     B=IsNamed(self=@n2.AnimInputEvent, InEventName=aie.EnemyPlayerEliminated)
     C:bool=AND Boolean(A=@n3, B=@n5)
    ) -> (ReturnValue))
   > Set ReactiveAnimation(ReactiveAnimation:bool=true) -> (Output_Get:bool)
   Event BlueprintPostEvaluateAnimation() {is=title K2Node_Event}
   > Set ReactiveAnimation(ReactiveAnimation=false)
   TryGetPawnOwner()
  ]}
]
variables [
 As Player Pawn Athena {type=PlayerPawn_Athena_C}
 Gravity Override Params {type=GravityOverrideParamsStruct}
 GravityOverride {type=Vector}
 ReactiveAnimation {type=boolean}
]
